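# Parent material for the BoldDorm Bagel gadget emote prop, a masked character material that clips marked geometry against a socket-defined plane with a glowing edge.
# It wraps the 2023 character PreFX and PostFX functions, overriding opacity mask and emissive between them.
# Where vertex red is set, a dot product against SocketPos and SocketNormal hides pixels behind the plane.
# An offset plane, shaped by thickness and exponent parameters, adds a cyan glow at the cut.
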Material /BRCosmetics/Animation/Props/Gadgets_BoldDorm_Bagel/Materials/M_BoldDorm_Bagel_Emote_Prop_Parent
graph {
 properties {
  bUseMaterialAttributes=true
  bUsedWithMeshDeformer=true
  bUsedWithMorphTargets=true
  bUsedWithSkeletalMesh=true
  blendMode=BLEND_Masked
  pixelDepthOffsetMode=PDOM_Legacy
  shadingModel=MSM_FromMaterialExpression
  usedShadingModels=Default Lit | Subsurface | Clear Coat | Subsurface Profile | Cloth
 }
 nodes [
  n1 = MaterialFunctionCall() {materialFunction=&MF_CharacterParent_2023_PreFX}
  n2 = VertexColor()
  n3 = OneMinus(None=@n2.R)
  n4 = VectorParameter() {parameterName=SocketPos, defaultValue=<rgba 0 0 0 1>}
  n5 = Normalize(VectorInput=Subtract(A=WorldPosition().XYZ, B=@n4.RGB))
  n6 = VectorParameter() {parameterName=SocketNormal, defaultValue=<rgba 0 1 0 1>}
  n7 = Saturate(None=Ceil(None=DotProduct(A=@n5, B=@n6.RGB)))
  n8 = GetMaterialAttributes(None=@n1.Result) {attributeGetTypes [#1]}
  n9 = Multiply(A=@n6.RGB, B=ScalarParameter() {defaultValue=0.5, parameterName=MaskGlowThickness})
  n10 = Normalize(VectorInput=Subtract(A=WorldPosition().XYZ, B=Add(A=@n4.RGB, B=@n9)))
  n11 = OneMinus(None=Saturate(None=DotProduct(A=@n10, B=@n6.RGB)))
  n12 = ScalarParameter() {defaultValue=50, parameterName=MaskGlowExp}
  n13 = Multiply(A=Multiply(A=Power(Base=@n11, Exp=@n12), B=@n2.R), B=Constant3Vector() {constant=<rgba 0.25 3.4167 5 1>})
  MaterialFunctionCall(
   Material=SetMaterialAttributes!(MaterialAttributes=@n1.Result, Opacity Mask=Add(A=@n3, B=@n7), Emissive Color=Add(A=@n8.Emissive Color, B=@n13)) {attributeSetTypes [#2, #1]}
   M=@n1.M
   Pre Skinned Local Position=@n1.M
   Distance Blend=@n1.M
   Fresnel=@n1.M
  ) {materialFunction=&MF_CharacterParent_2023_PostFX}
 ]
}
